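# Translucent, unlit, two-sided Niagara material for the source mesh of the Armadillo robot emote's chest beam.
# A view-angle fade, with sensitivity varied by panning noise, blends three colors, which particle color and Brightness tint before an emissive adjustment function.
# Opacity multiplies particle alpha, Fresnel and an Opacity parameter by a squared V gradient that cuts off sharply near V=1.
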Material /BRCosmetics/Animation/Game/MainPlayer/Emotes/ArmadilloRobot/FX/M_Armadillo_ChestBeam_Mesh_SourceV2
graph {
 properties {
  bUsedWithNiagaraMeshParticles=true
  bUsedWithNiagaraSprites=true
  blendMode=BLEND_Translucent
  pixelDepthOffsetMode=PDOM_Legacy
  shadingModel=MSM_Unlit
  twoSided=true
  usedShadingModels=Unlit
 }
 nodes [
  n1 = ParticleColor()
  n2 = ScalarParameter() {defaultValue=2, parameterName=EdgeNoiseA}
  n3 = ScalarParameter() {defaultValue=4, parameterName=EdgeNoiseB}
  n4 = TextureCoordinate()
  n5 = TextureSample(UVs=Panner(Coordinate=Multiply(A=@n4) {constB=2}) {speedX=0.1, speedY=1}) {
   samplerType=SAMPLERTYPE_LinearGrayscale
   texture=/BRCosmetics/Athena/Effects/Particles/Skydiving/NS_Contrail_DarkAndLight/T_DarkAndLight_Noise
  }
  n6 = MaterialFunctionCall(
   A=VectorParameter() {parameterName=ColorC, defaultValue=<rgba 1 0.05 0.099 1>}.RGB
   B=VectorParameter() {parameterName=ColorB, defaultValue=<rgba 0.1719 0.0313 0.0489 1>}.RGB
   C=VectorParameter() {parameterName=ColorA, defaultValue=<rgba 1 0 0.0398 1>}.RGB
   Alpha=MaterialFunctionCall("FadeSensitivity (optional)"=LinearInterpolate(A=@n2, B=@n3, Alpha=@n5.R) {constA=2, constB=4}) {materialFunction=&FadeBasedOnViewAngle}.Result
  ) {materialFunction=&3ColorBlend}
  MaterialFunctionCall(Emissive=Multiply(A=ScalarParameter() {defaultValue=1, parameterName=Brightness}, B=Multiply(A=@n1.RGB, B=@n6.Result))) {materialFunction=&MF_FX_EmissiveAdjustments}
  n7 = Fresnel() {baseReflectFraction=0.01}
  n8 = MaterialFunctionCall(Float2=TextureCoordinate()) {materialFunction=&BreakOutFloat2Components}
  n9 = Multiply(A=@n8.G, B=@n8.G)
  n10 = Saturate(None=Multiply(A=OneMinus(None=@n8.G)) {constB=10})
  Multiply(A=Multiply(A=Multiply(A=@n1.A, B=@n7), B=Multiply(A=@n9, B=@n10)), B=ScalarParameter() {defaultValue=1.7787, parameterName=Opacity}) {constA=2, constB=1.5}
 ]
}
