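# Additive, unlit, two-sided Niagara mesh-particle material that draws chromatic light rays for the ArmadilloRobot emote FX.
# Curve-atlas ramps shape the rays and sum two color lookups, then tint, scrolling electric noise and a sine mask form the emissive.
# Noise scaled along V builds an instance-space offset for ray sway and spread.
# Dynamic parameters drive motion, scale, desaturation and exponent, while particle alpha feeds depth fade, skipped on mobile.
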Material /BRCosmetics/Animation/Game/MainPlayer/Emotes/ArmadilloRobot/FX/M_Armadillo_ChromaticLightRay
graph {
 properties {
  bUsedWithNiagaraMeshParticles=true
  blendMode=BLEND_Additive
  pixelDepthOffsetMode=PDOM_Legacy
  shadingModel=MSM_Unlit
  twoSided=true
  usedShadingModels=Unlit
 }
 nodes [
  n1 = Constant2Vector() {g=0.0051, r=0.05}
  n2 = DynamicParameter() {parameterIndex=1, defaultValue=<rgba 1 1 1 0>, paramNames [ColorExp, MontionStrength, Scale, Desaturation]}
  n3 = Add(A=@n2.MontionStrength, B=DynamicParameter() {paramNames [Time,Random,Param3,Param4]}.Random)
  n4 = TextureCoordinate()
  n5 = Add(A=ComponentMask(None=@n4) {r=true}, B=Multiply(A=@n3, B=ScalarParameter() {defaultValue=-2.3, parameterName=WPOTimeScale}))
  n6 = AppendVector(
   A=Multiply(A=@n5, B=ScalarParameter() {defaultValue=2, parameterName=LightRayTiling}) {constB=2}
   B=ComponentMask(None=TextureCoordinate() {vTiling=0.01}) {g=true}
  )
  n7 = TextureCoordinate()
  n8 = CurveAtlasRowParameter(CurveTime=ComponentMask(None=@n7) {r=true}) {atlas=/Game/Athena/LinearColorCurves/TempAthenaCurveAtlas, curve=/Game/Athena/Environments/Nexus/Materials/FXshader/Curve_Ramp2}
  n9 = CurveAtlasRowParameter(CurveTime=@n8.B) {
   atlas=&AthenaCurveAtlas01
   curve=/Game/Athena/Environments/Nexus/Materials/FXshader/Curve_Color
   defaultValue=78
   parameterName=Param2
  }
  n10 = CurveAtlasRowParameter(CurveTime=@n8.A) {
   atlas=&AthenaCurveAtlas01
   curve=&Curve_Color
   defaultValue=78
   parameterName=Param2
  }
  n11 = VectorParameter() {group=Param, parameterName=ColorTint, defaultValue=<rgba 0.1639 0.4298 0.5 1>}
  n12 = DynamicParameter() {parameterIndex=1, defaultValue=<rgba 1 1 1 0>, paramNames [ColorExp, MontionStrength, Scale, Desaturation]}
  n13 = AppendVector(A=Desaturation(None=@n11.RGB, Fraction=@n12.Desaturation), B=@n11.A)
  n14 = Multiply(A=Add(A=@n9, B=@n10), B=ComponentMask(None=@n13) {b=true, g=true, r=true})
  n15 = ComponentMask(None=@n7) {g=true}
  n16 = Multiply(
   A=Multiply(
    A=CurveAtlasRowParameter(CurveTime=@n15) {atlas=&TempAthenaCurveAtlas, curve=&Curve_Ramp2}.R
    B=@n8.G
   )
   B=ComponentMask(None=@n13) {a=true}
  )
  n17 = ParticleColor()
  n18 = Multiply(
   A=TextureSample(UVs=Panner(Coordinate=@n6, Time=@n3, Speed=@n1)) {samplerType=SAMPLERTYPE_LinearColor, texture=/Game/Characters/Player/Male/Medium/Bodies/M_MED_SRDrift/Textures/T_ElectricNoise}.R
   B=Multiply(A=Multiply(A=@n14, B=@n16), B=@n17.RGB)
  )
  n19 = DynamicParameter() {parameterIndex=1, defaultValue=<rgba 1 1 1 0>, paramNames [ColorExp, MontionStrength, Scale, Desaturation]}
  n20 = Power(Base=@n18, Exp=@n19.ColorExp)
  n21 = TextureCoordinate()
  n22 = Sine(None=Divide(A=MaterialFunctionCall(Float2=@n21) {materialFunction=&BreakOutFloat2Components}.R))
  n23 = Power(Base=@n22)
  MaterialFunctionCall(Emissive=Multiply(A=Max(A=@n20, B=@n18), B=Saturate(None=@n23))) {materialFunction=&MF_FX_EmissiveAdjustments}
  Multiply(A=@n22, B=ScalarParameter() {defaultValue=5, parameterName=Param_1}) {constB=3}
  ShadingPathSwitch(Default=DepthFade(Opacity=@n17.A, FadeDistance=ScalarParameter() {defaultValue=1000, parameterName=DF_FD}), Mobile=@n17.A) {desc=Migrated from FeatureLevelSwitch}
  n24 = TextureSample(UVs=AppendVector(A=@n5, B=Constant())) {samplerType=SAMPLERTYPE_LinearColor, texture=&HighContrastNoise}
  n25 = LinearInterpolate(Alpha=ComponentMask(None=@n4) {g=true}) {constB=5}
  n26 = ScalarParameter() {defaultValue=50, parameterName=Wide}
  n27 = ScalarParameter() {defaultValue=-15, parameterName=X}
  n28 = ScalarParameter() {defaultValue=25, parameterName=Y}
  n29 = ScalarParameter() {defaultValue=1, parameterName=Z_Random}
  n30 = Multiply(A=LinearInterpolate(Alpha=ParticleRandom()) {constA=3, constB=-10}, B=@n29)
  n31 = DynamicParameter() {parameterIndex=1, defaultValue=<rgba 1 1 1 0>, paramNames [ColorExp, MontionStrength, Scale, Desaturation]}
  n32 = Add(
   A=@n26
   B=Multiply(
    A=MaterialFunctionCall(R=@n27, G=@n28, B=@n29, A=@n30) {materialFunction=&AppendMany}.RGB
    B=@n31.Scale
   )
  )
  n33 = ComponentMask(None=TextureCoordinate()) {g=true}
  Transform(None=Multiply(A=Multiply(A=@n24.RGB, B=@n25), B=LinearInterpolate(A=@n26, B=@n32, Alpha=@n33))) {transformSourceType=TRANSFORMSOURCE_Instance}
 ]
}
